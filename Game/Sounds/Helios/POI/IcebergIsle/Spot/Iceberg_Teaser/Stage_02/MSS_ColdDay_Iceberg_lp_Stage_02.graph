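# Stereo looping MetaSound source for the stage 2 cold day Iceberg Isle teaser spot sound in Helios.
# A single Wave Player starts the iceberg teaser loop wave on play and feeds its left and right channels straight to the stereo outputs.
# It uses a stage-specific attenuation asset and the in-game SFX sound class.
# A concurrency override allows one instance at a time and blocks new ones.
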MetaSoundSource /Game/Sounds/Helios/POI/IcebergIsle/Spot/Iceberg_Teaser/Stage_02/MSS_ColdDay_Iceberg_lp_Stage_02
inputs [
 Iceberg Sound {type=WaveAsset}
 UE.Source.OnPlay {type=Trigger}
]
interfaces [UE.OutputFormat.Stereo:1.0, UE.Source:1.1]
outputs [
 "UE.OutputFormat.Stereo.Audio:0" {type=Audio}
 "UE.OutputFormat.Stereo.Audio:1" {type=Audio}
]
pages [Default {
  input_defaults {Iceberg Sound=/Game/Sounds/Helios/POI/IcebergIsle/Spot/Iceberg_Teaser/wav/FN_S29_Helios_ColdDay_IceTeaser_lp, UE.Source.OnPlay=null}
  nodes [WavePlayer_0 = UE.Wave Player:Stereo(Wave Asset=^Iceberg Sound, Play=^UE.Source.OnPlay) {also {defaults {Loop="<MetaSoundBuilderResult.SUCCEEDED: 0>"}}}]
  links [{dst="@UE.OutputFormat.Stereo.Audio:0", src=WavePlayer_0.Out Left}, {dst="@UE.OutputFormat.Stereo.Audio:1", src=WavePlayer_0.Out Right}]
 }]
properties {
 assetImportData=&self:AssetImportData
 attenuationSettings=&/Att_IcebergTeaser_Stage_02
 bLooping=true
 bOverrideConcurrency=true
 duration=10000
 maxDistance=26855.332
 numChannels=2
 outputFormat=Stereo
 sampleRate=48000
 soundClassObject=&Fort_SFX_InGame
 concurrencyOverrides {maxCount=1, resolutionRule=PreventNew, retriggerTime=1}
}
